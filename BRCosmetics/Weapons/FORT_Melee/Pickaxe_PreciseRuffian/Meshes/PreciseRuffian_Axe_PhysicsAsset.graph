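# Physics asset for the PreciseRuffian two-handed melee pickaxe mesh, a cosmetic weapon.
# It contains a single body on the Melee_TwoHanded_Handle bone, shaped as one capsule rotated along the handle's length.
# With no constraints, it provides one simple collision volume rather than any articulated simulation.
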PhysicsAsset /BRCosmetics/Weapons/FORT_Melee/Pickaxe_PreciseRuffian/Meshes/PreciseRuffian_Axe_PhysicsAsset
bodies {Melee_TwoHanded_Handle [{shapeType=Capsule, length=58.1148, radius=12.9558, center=<0.0272 -0.0038 37.6042>, rotation=<rot -0.0136 -90 0.028>}]}
constraints []
